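; Environmental spline zipline that players ride, strung between two optional pole actors.
; Construction can auto-curve the spline between pole sockets using lerped horizontal/vertical vectors with linear blending, then builds spline-mesh segments with dynamic materials and positions end interaction spheres.
; BeginPlay initializes the zipline from its endpoints and, if destruction is enabled, makes it die with either pole.
; Riding timestamps cable materials for VFX, blocking tags prevent interaction, and exiting can grant glider redeploy.
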(event EventBeginPlay
  (bind _ziplinespline (Variables|FortAthenaSplineZipline|GetZiplineSpline))
  (bind _returnvalue (Spline|GetNumberOfSplinePoints _ziplinespline))
  (bind _outputdelegate (EventDispatchers|CreateEvent :!event Custom|HandlePoleDied))
  (Utilities|FlowControl|Sequence
    (:then_0
      ;[0.5103 0.6891 1] Allow destruction if DestructionEnabled is true
      (if (ScalableFloat|GetValueasBool (Variables|Zipline|GetDestructionEnabled))
        (EventDispatchers|AssignOnDied (Variables|Default|GetPoleA) _outputdelegate)
        (EventDispatchers|BindEventtoOnDied (Variables|Default|GetPoleB) _outputdelegate)))
    (:then_1
      (if (> _returnvalue 0)
        (Zipline|Initialize (Spline|GetLocationatSplinePoint _ziplinespline 0 "World") (Spline|GetLocationatSplinePoint _ziplinespline (- _returnvalue 1) "World"))))))

(event EventOnZipliningStarted
  (params (InteractingPawn :type Fort%20Player%20Pawn%20Object%20Reference) (InteractComponent :type Primitive%20Component%20Object%20Reference))
  ;[0.3209 0.3489 1] VFX for when local player is attached/detached from zipLine
  (|CallParentFunction InteractingPawn InteractComponent)
  (Networking|CanExecuteCosmeticEvents
    (:True
      (for _array_element (Variables|Default|GetSplineMaterials)
        (Rendering|Material|SetScalarParameterValue _array_element "TimeStamp" (Utilities|Time|GetGameTimeinSeconds) :!class /Script/Engine.MaterialInstanceDynamic)))))

(event EventOnZipliningStopped
  (params (InteractingPawn :type Fort%20Player%20Pawn%20Object%20Reference) (InteractComponent :type Primitive%20Component%20Object%20Reference) (ZiplineUsageDuration :type Float%20%28single-precision%29))
  ;[0.3209 0.3489 1] VFX for when local player is attached/detached from zipLine
  (|CallParentFunction InteractingPawn InteractComponent ZiplineUsageDuration)
  (Utilities|FlowControl|Sequence
    (:then_0
      (Networking|CanExecuteCosmeticEvents
        (:True
          (for _array_element (Variables|Default|GetSplineMaterials)
            (Rendering|Material|SetScalarParameterValue _array_element "TimeStamp" (Utilities|Time|GetGameTimeinSeconds) :!class /Script/Engine.MaterialInstanceDynamic)))))
    (:then_1
      (CallFunction|GliderRedeploy InteractingPawn))))

(event Custom|HandlePoleDied
  (params (DamagedActor :type Actor%20Object%20Reference) (Damage :type Float%20%28single-precision%29) (InstigatedBy :type Controller%20Object%20Reference) (DamageCauser :type Actor%20Object%20Reference) (HitLocation :type Vector) (FHitComponent :type Primitive%20Component%20Object%20Reference) (BoneName :type Name) (Momentum :type Vector))
  (bind _outputdelegate (EventDispatchers|CreateEvent :!event Custom|HandlePoleDied))
  ;[1 1 1] Die when one of the poles dies
  (EventDispatchers|UnbindEventfromOnDied (Variables|Default|GetPoleA) _outputdelegate)
  (EventDispatchers|UnbindEventfromOnDied (Variables|Default|GetPoleB) _outputdelegate)
  (Rendering|SetActorHiddenInGame true)
  (Gameplay|Die))

(fn AddSplineMeshSegment
  (bind _returnvalue (AddComponent|NODE_AddSplineMeshComponent-2))
  (Components|StaticMesh|SetStaticMesh _returnvalue (Variables|Default|GetSplineStaticMesh))
  (SplineMesh|SetForwardAxis _returnvalue (Variables|Default|GetForwardMeshAxis) true)
  (Collision|SetCollisionProfileName _returnvalue "FortTriggerOnlyInteractions" false :!class /Script/Engine.PrimitiveComponent)
  (Variables|/Script/Engine.PrimitiveComponent|SetbGenerateOverlapEvents true _returnvalue)
  (return _returnvalue))

(fn AutoSmoothTanget
  (params (Tangent :type Vector) (PointA :type Vector) (PointB :type Vector))
  (return (select (Variables|TangentSmoothing|GetAutoSmoothTangents) (* (* (Math|Vector|Normalize Tangent) (Math|Vector|VectorLength (- PointA PointB))) (Variables|TangentSmoothing|GetTangentSmoothStrength)) Tangent)))

(fn BlueprintCanInteract
  (params (InteractingPawn :type Fort%20Pawn%20Object%20Reference) (InteractionBeingAttempted :type EInteractionBeingAttempted%20Enum :default "FirstInteraction") (InteractionType :type TInteractionType%20Enum :default "IT_NoInteraction"))
  (return (not (GameplayTags|HasAnyMatchingGameplayTags InteractingPawn (Variables|Default|GetBlockInteractTags)))))

(fn CalcAutoLocationAtAlpha
  (params (InAlpha :type Float%20%28double-precision%29) (DrawDebug :type Boolean))
  (bind _lowerendlocation (Variables|Default|GetLowerEndLocation))
  (bind _higherendlocation (Variables|Default|GetHigherEndLocation))
  (bind _returnvalue (- _higherendlocation _lowerendlocation))
  (bind _inalpha (Variables|GetInAlpha))
  (bind _inalpha_2 (Variables|GetInAlpha))
  (bind _lowerendlocation_2 (Variables|Default|GetLowerEndLocation))
  (bind _higherendlocation_2 (Variables|Default|GetHigherEndLocation))
  (bind _lowerendlocation_3 (Variables|Default|GetLowerEndLocation))
  (bind _inalpha_3 (Variables|GetInAlpha))
  ;[1 1 1] Calc Auto Location for alpha/factor along spline
  (bind (_vertvec _horizvec) (CallFunction|GetAutoHorizandVertVectors (Variables|Default|GetHigherEndLocation) (Variables|Default|GetLowerEndLocation)))
  (if DrawDebug
    ;[1 1 1] DebugDisplay
    (Rendering|Debug|DrawDebugLine (Variables|Default|GetHigherEndLocation) (+ (Variables|Default|GetLowerEndLocation) (- (- _higherendlocation_2 (* _vertvec -1.0)) _lowerendlocation_3)) "(R=1.000000,G=0.921126,B=0.712643,A=1.000000)" 5.0 4.0)
    (Rendering|Debug|DrawDebugLine (Variables|Default|GetLowerEndLocation) (+ (Variables|Default|GetLowerEndLocation) (* _horizvec -1.0)) "(R=0.685079,G=0.707206,B=1.000000,A=1.000000)" 5.0 4.0)
    (Rendering|Debug|DrawDebugLine (Math|Vector|Lerp(Vector) (Math|Vector|Lerp(Vector) "0, 0, 0" (* _horizvec -1.0) _inalpha) (Math|Vector|Lerp(Vector) (- (- _higherendlocation_2 (* _vertvec -1.0)) _lowerendlocation_3) _returnvalue _inalpha_3) _inalpha_2) (+ (Math|Vector|Lerp(Vector) (Math|Vector|Lerp(Vector) "0, 0, 0" (* _horizvec -1.0) _inalpha) (Math|Vector|Lerp(Vector) (- (- _higherendlocation_2 (* _vertvec -1.0)) _lowerendlocation_3) _returnvalue _inalpha_3) _inalpha_2) "0, 0,500.000000") "(R=1.000000,G=0.000000,B=0.000000,A=1.000000)" 5.0 4.0))
  (if (or (> (Variables|Default|GetAutoLinearFactorHigh) 0.0) (> (Variables|Default|GetAutoLinearFactorLow) 0.0))
    ;[1 1 1] Lerp to the linear version of the point
    (return (Math|Vector|Lerp(Vector) (+ _lowerendlocation_2 (Math|Vector|Lerp(Vector) (Math|Vector|Lerp(Vector) "0, 0, 0" (* _horizvec -1.0) _inalpha) (Math|Vector|Lerp(Vector) (- (- _higherendlocation_2 (* _vertvec -1.0)) _lowerendlocation_3) _returnvalue _inalpha_3) _inalpha_2)) (Math|Vector|Lerp(Vector) (Variables|Default|GetLowerEndLocation) (Variables|Default|GetHigherEndLocation) (Variables|GetInAlpha)) (Math|Float|Clamp(Float) (Math|Float|Lerp (Variables|Default|GetAutoLinearFactorHigh) (Variables|Default|GetAutoLinearFactorLow) (Variables|GetInAlpha)))))
    (else
      (return (+ _lowerendlocation_2 (Math|Vector|Lerp(Vector) (Math|Vector|Lerp(Vector) "0, 0, 0" (* _horizvec -1.0) _inalpha) (Math|Vector|Lerp(Vector) (- (- _higherendlocation_2 (* _vertvec -1.0)) _lowerendlocation_3) _returnvalue _inalpha_3) _inalpha_2))))))

(fn CalculatePositionOfPoles
  (bind _polea (Variables|Default|GetPoleA))
  (bind _poleb (Variables|Default|GetPoleB))
  (Utilities|IsValid _polea
    (:Is%20Valid
      (Variables|Default|SetPoleASocketLocation (Transformation|GetSocketLocation (Variables|/Script/Engine.Actor|GetRootComponent _polea) "LineAttachSocket"))
      (Utilities|IsValid _poleb
        (:Is%20Valid
          (Variables|Default|SetPoleBSocketLocation (Transformation|GetSocketLocation (Variables|/Script/Engine.Actor|GetRootComponent _poleb) "LineAttachSocket")))
        (:Is%20Not%20Valid
          (Variables|Default|SetPoleBSocketLocation (Spline|GetLocationandTangentatSplinePoint (Variables|FortAthenaSplineZipline|GetZiplineSpline) (- (Spline|GetNumberOfSplinePoints (Variables|FortAthenaSplineZipline|GetZiplineSpline)) 1) "Local")))))
    (:Is%20Not%20Valid
      (Variables|Default|SetPoleASocketLocation (Spline|GetLocationandTangentatSplinePoint (Variables|FortAthenaSplineZipline|GetZiplineSpline) 0 "Local")))))

(fn CreateSplineMesh
  (params (FirstPoint :type Integer) (SecondPoint :type Integer))
  (bind (_location _tangent) (Spline|GetLocationandTangentatSplinePoint (Variables|FortAthenaSplineZipline|GetZiplineSpline) FirstPoint "Local"))
  (bind (_location_2 _tangent_2) (Spline|GetLocationandTangentatSplinePoint (Variables|FortAthenaSplineZipline|GetZiplineSpline) SecondPoint "Local"))
  (bind _splinemeshsegment (CallFunction|AddSplineMeshSegment))
  (bind _output_get (LocalVariables|SetSpawnedSplineMeshComponent _splinemeshsegment))
  (SplineMesh|SetStartandEnd _output_get _location (CallFunction|AutoSmoothTanget _tangent _location _location_2) _location_2 (CallFunction|AutoSmoothTanget _tangent_2 _location _location_2) true :!class /Script/Engine.SplineMeshComponent)
  (CallFunction|UpdateSplineMeshMaterials (LocalVariables|GetSpawnedSplineMeshComponent) (Variables|GetFirstPoint)))

(fn GetAutoHorizAndVertVectors
  (params (highVector :type Vector) (LowVector :type Vector))
  (bind (_returnvalue_x _returnvalue_y _returnvalue_z) (Math|Vector|BreakVector (- highVector LowVector)))
  (return (Math|Vector|MakeVector 0.0 0.0 (* _returnvalue_z (Variables|Default|GetAutoSplineTangentVertCoef))) (* (Math|Vector|MakeVector _returnvalue_x _returnvalue_y) (Variables|Default|GetAutoSplineTangentHorizCoef))))

(fn GetMIDForSplineMeshComp
  (params (SplineMeshComp :type Spline%20Mesh%20Component%20Object%20Reference) (ElementIndex :type Integer))
  (bind _returnvalue (Rendering|Material|CreateDynamicMaterialInstance SplineMeshComp ElementIndex 0 "None" :!class /Script/Engine.PrimitiveComponent))
  (return _returnvalue))

(fn GetMotorOffset
  (return (Variables|Default|GetMotorOffset)))

(fn GliderRedeploy
  (params (Fort%20Player%20Pawn :type Fort%20Player%20Pawn%20Object%20Reference))
  ;[0.9531 0 0.6833 0.5] Glider Redeploy
  (if (ScalableFloat|GetValueasBool (Variables|Default|GetHF_AllowGliderRedeployOnExitZipline))
    (bind _returnvalue (GameplayEffects|ApplyGameplayEffectToSelf :self (Ability|GetAbilitySystemComponent Fort%20Player%20Pawn) :GameplayEffectClass (Variables|Default|GetGlider%20Redeploy%20Gameplay%20Effect%20Class) :Level 1.0))
    (return _returnvalue)))

(fn SetCableTilingBySplineLength
  (params (Target :type Mesh%20Component%20Object%20Reference))
  (Rendering|Material|SetScalarParameterValueonMaterials Target "Tile_V" (Math|Conversions|ToFloat(Integer) (Math|Float|Truncate (/ (Spline|GetSplineLength (Variables|FortAthenaSplineZipline|GetZiplineSpline)) (Variables|Default|GetTilingDivisor))))))

(fn SetInteractionCollision
  (bind _ziplinespline (Variables|FortAthenaSplineZipline|GetZiplineSpline))
  (Transformation|SetWorldLocation (Variables|Default|GetStartPointInteractSphere) (Spline|GetLocationatSplinePoint _ziplinespline 0 "World") false true)
  (Transformation|SetWorldLocation (Variables|Default|GetEndPointInteractSphere) (Spline|GetLocationatSplinePoint _ziplinespline (- (Spline|GetNumberOfSplinePoints _ziplinespline) 1) "World") false true))

(fn SetSplinePositionAndTangent
  (params (SetPosition :type Boolean) (SetTangent :type Boolean) (ID :type Integer))
  (bind _ziplinespline (Variables|FortAthenaSplineZipline|GetZiplineSpline))
  (bind _returnvalue (Spline|GetNumberOfSplinePoints _ziplinespline))
  (bind _returnvalue_2 (- _returnvalue 1))
  (bind _returnvalue_3 (Math|Conversions|ToFloat(Integer) _returnvalue_2))
  (bind _alphaforlocation (LocalVariables|GetAlphaForLocation))
  (bind _onedivnumpoints (LocalVariables|GetOneDivNumPoints))
  (bind _onedivnumpoints_2 (LocalVariables|GetOneDivNumPoints))
  (bind _autosplinetangentlengthcoef (Variables|Default|GetAutoSplineTangentLengthCoef))
  ;[1 1 1] Set up alphas
  (bind _output_get (LocalVariables|SetSplinePointID ID))
  (LocalVariables|SetSplinePointIDToSet _output_get)
  (if (== (LocalVariables|GetSplinePointID) 0)
    (LocalVariables|SetSplinePointID 1)
    (LocalVariables|SetIsFirstPoint true))
  (if (and (> (LocalVariables|GetSplinePointID) 0) (> _returnvalue 1))
    (LocalVariables|SetOneDivNumPoints (/ 1.0 _returnvalue_3))
    (LocalVariables|SetAlpha (/ (Math|Conversions|ToFloat(Integer) (LocalVariables|GetSplinePointID)) _returnvalue_3))
    (LocalVariables|SetPrevAlpha (/ (Math|Conversions|ToFloat(Integer) (- (LocalVariables|GetSplinePointID) 1)) _returnvalue_3))
    (if (LocalVariables|GetIsFirstPoint)
      (LocalVariables|SetAlphaForLocation (LocalVariables|GetPrevAlpha))
      (else
        (LocalVariables|SetAlphaForLocation (LocalVariables|GetAlpha))))
    ;[1 1 1] Calc Auto Location for point
    (bind _pointlocation (CallFunction|CalcAutoLocationatAlpha (LocalVariables|GetAlphaForLocation)))
    (if (Variables|GetSetPosition)
      (Spline|SetLocationatSplinePoint (Variables|FortAthenaSplineZipline|GetZiplineSpline) (LocalVariables|GetSplinePointIDToSet) _pointlocation "World" false)
      (else
        _))
    ;[1 1 1] Calc Auto Tangent for point
    (if (Variables|GetSetTangent)
      (bind (_vertvec _horizvec) (CallFunction|GetAutoHorizandVertVectors (Variables|Default|GetHigherEndLocation) (Variables|Default|GetLowerEndLocation)))
      (if (or (> (Variables|Default|GetAutoLinearFactorHigh) 0.0) (> (Variables|Default|GetAutoLinearFactorLow) 0.0))
        ;[1 1 1] Lerp to linear version
        (Spline|SetTangentatSplinePoint (Variables|FortAthenaSplineZipline|GetZiplineSpline) (LocalVariables|GetSplinePointIDToSet) (Math|Vector|Lerp(Vector) (* (Math|Vector|Lerp(Vector) (* _vertvec _onedivnumpoints) (* _horizvec _onedivnumpoints_2) _alphaforlocation) _autosplinetangentlengthcoef) (* (- (Variables|Default|GetHigherEndLocation) (Variables|Default|GetLowerEndLocation)) (LocalVariables|GetOneDivNumPoints)) (Math|Float|Clamp(Float) (Math|Float|Lerp (Variables|Default|GetAutoLinearFactorHigh) (Variables|Default|GetAutoLinearFactorLow) (LocalVariables|GetAlphaForLocation)))) "World" false)
        (else
          (Spline|SetTangentatSplinePoint (Variables|FortAthenaSplineZipline|GetZiplineSpline) (LocalVariables|GetSplinePointIDToSet) (* (Math|Vector|Lerp(Vector) (* _vertvec _onedivnumpoints) (* _horizvec _onedivnumpoints_2) _alphaforlocation) _autosplinetangentlengthcoef) "World" false))))))

(fn UpdateSplineMeshMaterials
  (params (SplineMeshComp :type Spline%20Mesh%20Component%20Object%20Reference) (SplineIndex :type Integer))
  ;[1 1 1] Set Material Instances for all splines Start and End Falloff
  (for _index (range (Rendering|Material|GetNumMaterials SplineMeshComp))
    (bind _mid (CallFunction|GetMIDForSplineMeshComp (Variables|GetSplineMeshComp) _index))
    (Rendering|Material|SetScalarParameterValue _mid "UseStart" (Math|Float|SelectFloat 1.0 0.0 (== (- (Spline|GetNumberOfSplinePoints (Variables|FortAthenaSplineZipline|GetZiplineSpline)) 2) (Variables|GetSplineIndex))) :!class /Script/Engine.MaterialInstanceDynamic)
    (Rendering|Material|SetScalarParameterValue _mid "UseEnd" (Math|Float|SelectFloat 1.0 0.0 (== 0 (Variables|GetSplineIndex))) :!class /Script/Engine.MaterialInstanceDynamic)
    (Utilities|Array|Add (Variables|Default|GetSplineMaterials) _mid)
    (CallFunction|SetCableTilingbySplineLength (Variables|GetSplineMeshComp))))

(fn ConstructionScript
  (bind _ziplinespline (Variables|FortAthenaSplineZipline|GetZiplineSpline))
  (bind _returnvalue (Spline|GetNumberOfSplinePoints _ziplinespline))
  (bind _returnvalue_2 (- _returnvalue 1))
  (bind (_poleasocketlocation_x _poleasocketlocation_y _poleasocketlocation_z) (Math|Vector|BreakVector (Variables|GetPoleASocketLocation_X)))
  (bind (_polebsocketlocation_x _polebsocketlocation_y _polebsocketlocation_z) (Math|Vector|BreakVector (Variables|GetPoleBSocketLocation_X)))
  (bind _polea (Variables|Default|GetPoleA))
  (bind _poleb (Variables|Default|GetPoleB))
  (Utilities|IsValid _polea
    (:Is%20Valid
      (Variables|Default|SetPoleASocketLocation (Transformation|GetSocketLocation (Variables|/Script/Engine.Actor|GetRootComponent _polea) "LineAttachSocket"))
      (Utilities|IsValid _poleb
        (:Is%20Valid
          (Variables|Default|SetPoleBSocketLocation (Transformation|GetSocketLocation (Variables|/Script/Engine.Actor|GetRootComponent _poleb) "LineAttachSocket"))
          (if (< _poleasocketlocation_z _polebsocketlocation_z)
            (Variables|Default|SetLowerPointID 0)
            (Variables|Default|SetHigherPointID (- (Spline|GetNumberOfSplinePoints (Variables|FortAthenaSplineZipline|GetZiplineSpline)) 1))
            (Variables|Default|SetHigherEndLocation (Variables|Default|GetPoleASocketLocation))
            (Variables|Default|SetLowerEndLocation (Variables|Default|GetPoleBSocketLocation))
            (else
              (Variables|Default|SetHigherPointID 0)
              (Variables|Default|SetLowerPointID (- (Spline|GetNumberOfSplinePoints (Variables|FortAthenaSplineZipline|GetZiplineSpline)) 1))
              (Variables|Default|SetLowerEndLocation (Variables|Default|GetPoleASocketLocation))
              (Variables|Default|SetHigherEndLocation (Variables|Default|GetPoleBSocketLocation))))
          (for _index (range (+ _returnvalue_2 1))
            (if (or (== _index 0) (== _index _returnvalue_2))
              (CallFunction|SetSplinePositionandTangent (Variables|Default|GetAuto%20Set%20Spline%20Ends) (Variables|Default|GetAuto%20Set%20Spline%20Ends) _index)
              (else
                (CallFunction|SetSplinePositionandTangent (Variables|Default|GetAuto%20Set%20Spline%20Mids) (Variables|Default|GetAuto%20Set%20Spline%20Mids) _index))))
          (Spline|UpdateSpline (Variables|FortAthenaSplineZipline|GetZiplineSpline)))))
    (:Is%20Not%20Valid
      _))
  (CallFunction|SetInteractionCollision)
  (for _index_2 (range (+ (- (Spline|GetNumberOfSplinePoints (Variables|FortAthenaSplineZipline|GetZiplineSpline)) 2) 1))
    (Default|CreateSplineMesh _index_2 (+ _index_2 1))))
