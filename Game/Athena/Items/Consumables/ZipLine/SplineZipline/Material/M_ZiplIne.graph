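# Clear-coat metallic cable material for the spline-mesh zipline, with a wobble that radiates around the player's position.
# World position offset is a time-scrolled sine of distance to the player, scaled by near and far falloffs, a timestamp envelope and UV masks damping both ends.
# A static switch adds SplineThicken offset and normals.
# Tiled cable textures supply normal and color, with optional contrast-driven color remapping; roughness varies by shading path and quality.
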Material /Game/Athena/Items/Consumables/ZipLine/SplineZipline/Material/M_ZiplIne
graph {
 properties {bUsedWithSplineMeshes=true, pixelDepthOffsetMode=PDOM_Legacy, shadingModel=MSM_ClearCoat, usedShadingModels=Clear Coat}
 nodes [
  n1 = ScalarParameter() {defaultValue=-10, parameterName=Width}
  n2 = MaterialFunctionCall(Float2=TextureCoordinate()) {materialFunction=&BreakOutFloat2Components}
  n3 = MaterialFunctionCall(WidthBase=@n1, WidthTip=@n1, UVs For NormalShading=AppendVector(A=OneMinus(None=@n2.R), B=@n2.G)) {materialFunction=&SplineThicken}
  n4 = StaticSwitchParameter(True=@n3.WorldPositionOffset, False=Constant3Vector()) {parameterName=IsSplineThickening}
  n5 = Multiply(A=Time(), B=ScalarParameter() {defaultValue=-2, parameterName=Speed})
  n6 = VectorParameter() {parameterName=PlayerLocation, defaultValue=<rgba -6620 -8661 2870 1>}
  n7 = CollectionParameter() {collection=&FortniteMaterialParameters, parameterName=PlayerPosition}
  n8 = ComponentMask(None=@n7) {b=true, g=true, r=true}
  n9 = Distance(
   A=Add(A=@n4, B=WorldPosition().XYZ)
   B=Add(A=StaticSwitchParameter(True=@n6.RGB, False=@n8) {parameterName=Debug}, B=VectorParameter() {parameterName=PlayerPositionBias}.RGB)
  )
  n10 = Divide(A=@n9, B=VectorParameter() {parameterName=Frequency, defaultValue=<rgba 1134 1134 1134 0>}.RGB)
  n11 = VectorParameter() {parameterName=Amplitude, defaultValue=<rgba 0 0 15 0>}
  n12 = ScalarParameter() {defaultValue=8000, parameterName=DistanceFalloff}
  n13 = OneMinus(None=Saturate(None=Divide(A=@n9, B=@n12)))
  n14 = Divide(
   A=Subtract(A=@n9, B=ScalarParameter() {defaultValue=15.9454, parameterName=DistanceFalloffNearBias})
   B=ScalarParameter() {defaultValue=201.1301, parameterName=DistanceFalloffNear}
  )
  n15 = Multiply(A=Sine(None=Add(A=@n5, B=@n10)), B=Multiply(A=Multiply(A=@n11.RGB, B=@n13), B=Saturate(None=@n14)))
  n16 = ScalarParameter() {parameterName=TimestampA}
  n17 = ScalarParameter() {defaultValue=0.2, parameterName=TransitionLength_In}
  n18 = ScalarParameter() {defaultValue=1, parameterName=TimestampB}
  n19 = ScalarParameter() {defaultValue=0.5, parameterName=TransitionLength_Out}
  n20 = StaticBool()
  n21 = Time()
  n22 = ScalarParameter() {parameterName=TimeStamp}
  n23 = Saturate(None=Multiply(A=Subtract(A=@n21, B=@n22), B=ScalarParameter() {defaultValue=1, parameterName=DurationMultiplier}))
  n24 = StaticSwitchParameter(
   True=MaterialFunctionCall() {materialFunction=&DebugTimeSine}.Result
   False=StaticSwitchParameter(
    True=MaterialFunctionCall(TimestampA=@n16, TransitionLength_In=@n17, TimestampB=@n18, TransitionLength_Out=@n19, Transition In Only?=@n20) {materialFunction=&MF_Timestamp_InOut_01}.Linear
    False=Sine(None=@n23)
   ) {parameterName=Use_In/Out_Timestamps}
  ) {parameterName=Debug}
  n25 = ComponentMask(None=TextureCoordinate()) {g=true}
  n26 = ScalarParameter() {defaultValue=1, parameterName=UseStart}
  n27 = OneMinus(None=@n25)
  n28 = ScalarParameter() {defaultValue=1, parameterName=UseEnd}
  n29 = Multiply(
   A=ScalarParameter() {defaultValue=2, parameterName=FalloffStartEnd}
   B=Multiply(A=LinearInterpolate(B=@n25, Alpha=@n26) {constA=1}, B=LinearInterpolate(B=@n27, Alpha=@n28) {constA=1})
  )
  Add(A=@n4, B=Multiply(A=Multiply(A=@n15, B=@n24), B=Saturate(None=@n29)))
  ScalarParameter() {defaultValue=0.1, parameterName=Specular}
  ScalarParameter() {defaultValue=0.91, parameterName=Metallic}
  ScalarParameter() {defaultValue=0.3, group=Clear Coat, parameterName=Clear Coat}
  ScalarParameter() {defaultValue=0.4, group=Clear Coat, parameterName=Clear Coat Roughness}
  n30 = ScalarParameter() {defaultValue=0.45, parameterName=Roughness}
  n31 = ScalarParameter() {defaultValue=0.8, parameterName=Roughness Perf}
  ShadingPathSwitch(Default=@n30, Mobile=QualitySwitch(Default=@n31, Low=@n31, High=@n31, Medium=@n30, Epic=@n31))
  n32 = Multiply(
   A=TextureCoordinate()
   B=AppendVector(A=ScalarParameter() {defaultValue=2, parameterName=Tile_U}, B=ScalarParameter() {defaultValue=240, parameterName=Tile_V})
  )
  StaticSwitchParameter(
   True=Transform(None=@n3.Normal) {transformSourceType=TRANSFORMSOURCE_World, transformType=TRANSFORM_Tangent}
   False=Multiply(
    A=TextureSampleParameter2D(UVs=@n32) {parameterName=NormalMap, samplerType=SAMPLERTYPE_Normal, texture=/Game/Athena/Items/Consumables/ZipLine/T_Cable_N}.RGB
    B=VectorParameter() {parameterName=NormalStrength, defaultValue=<rgba 0.4 0.4 1 0>}.RGB
   )
  ) {parameterName=IsSplineThickening}
  n33 = TextureSampleParameter2D(UVs=@n32) {parameterName=BaseColor, texture=/Game/Athena/Items/Consumables/ZipLine/T_Cable}
  n34 = Multiply(
   A=Power(Base=@n33.RGB, Exp=ScalarParameter() {defaultValue=2, parameterName=Contrast})
   B=ScalarParameter() {defaultValue=1.3, parameterName=Brightness}
  )
  StaticSwitchParameter(
   True=LinearInterpolate(
    A=VectorParameter() {parameterName=BaseColorLow, defaultValue=<rgba 0.2199 0.2556 0.3142 0>}.RGB
    B=VectorParameter() {parameterName=BaseColorHigh, defaultValue=<rgba 0.4635 0.4635 0.4635 0>}.RGB
    Alpha=Saturate(None=@n34)
   )
   False=@n33.RGB
  ) {defaultValue=true, parameterName=UseBaseColorAdjustments}
 ]
}
